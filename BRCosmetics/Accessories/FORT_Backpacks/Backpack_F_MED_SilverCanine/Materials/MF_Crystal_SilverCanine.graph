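# Material function that gives the SilverCanine backpack a faked crystal gemstone look, overriding BaseColor and EmissiveColor on the incoming material attributes.
# A depth texture drives BumpOffset parallax for the tinted interior texture, with constant depth on low and mobile quality tiers.
# Fresnel, sky-light-direction backlight and skylight environment reflection terms add rim and glow, with emissive scaled by an Emissive texture.
# Three UV sphere masks form a SmallDiamondMask.
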MaterialFunction /BRCosmetics/Accessories/FORT_Backpacks/Backpack_F_MED_SilverCanine/Materials/MF_Crystal_SilverCanine
graph {nodes [
  n1 = VectorParameter() {group=Crystal, parameterName=BaseColorMultiply, sortPriority=0, defaultValue=<rgba 0.7 0.511 0.511 1>}
  n2 = Multiply(A=LinearInterpolate(A=@n1.RGB, Alpha=NamedRerouteUsage()), B=NamedRerouteUsage())
  n3 = TextureSampleParameter2D(UVs=NamedRerouteUsage()) {group=Crystal, parameterName=InteriorTexture, sortPriority=5, texture=/Game/Global/Textures/Default/Masks_and_Gradients/T_MF_Crystal}
  n4 = ScalarParameter() {defaultValue=0.75, group=Crystal, parameterName=InteriorPrismColors, sortPriority=2}
  n5 = Multiply(A=@n2, B=Desaturation(None=@n3.RGB, Fraction=OneMinus(None=@n4)))
  n6 = ScalarParameter() {defaultValue=10, group=Crystal, parameterName=InteriorIntensity, sortPriority=1}
  n7 = NamedRerouteDeclaration(None=LinearInterpolate(
    A=@n2
    B=Multiply(A=@n5, B=@n6) {constB=0.1}
    Alpha=ScalarParameter() {defaultValue=0.3, group=Crystal, parameterName=InteriorContrast, sortPriority=3}
   )) {name=InteriorColor, nodeColor=<rgba 1 0 0.4687 1>}
  n8 = LinearInterpolate(
   A=VectorParameter() {group=Crystal, parameterName=FresnelColor, sortPriority=19, defaultValue=<rgba 0 2 1.8658 1>}.RGB
   Alpha=NamedRerouteUsage()
  ) {constB=2}
  n9 = NamedRerouteUsage()
  n10 = Add(A=Add(A=NamedRerouteUsage(), B=Constant()), B=NamedRerouteUsage())
  n11 = TextureSampleParameter2D() {
   group=06_Emissive
   parameterName=Emissive
   samplerSource=SSM_Wrap_WorldGroupSettings
   sortPriority=1
   texture=&T_BlacksRGB
  }
  n12 = Multiply(A=@n11.RGB, B=ScalarParameter() {defaultValue=1.5, group=SliverCabineGemstone, parameterName=GemstoneEmissive})
  FunctionOutput(None=SetMaterialAttributes!(MaterialAttributes=NamedRerouteUsage(), BaseColor=LinearInterpolate(A=@n7, B=@n8, Alpha=@n9), EmissiveColor=Multiply(A=@n10, B=@n12)) {attributeSetTypes [#1, #2]})
  n13 = GetMaterialAttributes(None=FunctionInput() {inputName=Material, inputType=FunctionInput_MaterialAttributes}) {attributeGetTypes [#1, #2]}
  NamedRerouteDeclaration(None=@n13.MaterialAttributes) {name=MA, nodeColor=<rgba 0.3406 1 0.0156 1>}
  n14 = TextureSampleParameter2D(UVs=NamedRerouteUsage()) {
   group=Crystal
   parameterName=BacklightNormal
   samplerType=SAMPLERTYPE_Normal
   sortPriority=18
   texture=/Game/Global/Textures/Default/Masks_and_Gradients/T_MF_Crystal3
  }
  n15 = Transform(None=LinearInterpolate(
    A=Constant3Vector() {constant=<rgba 0 0 1 0>}
    B=@n14.RGB
    Alpha=ScalarParameter() {defaultValue=0.2, group=Crystal, parameterName=BacklightNormalIntensity, sortPriority=15}
   ) {constAlpha=0.1})
  n16 = CameraVectorWS()
  n17 = Multiply(A=ScalarParameter() {defaultValue=0.5, group=Crystal, parameterName=BacklightViewAngle, sortPriority=16}, B=@n16)
  n18 = DotProduct(A=Normalize(VectorInput=Add(A=@n15, B=@n17)), B=Multiply(A=SkyAtmosphereLightDirection()) {constB=-1})
  n19 = Saturate(None=Add(A=Multiply(A=@n18) {constB=0.5}) {constB=0.5})
  n20 = ScalarParameter() {defaultValue=2, group=Crystal, parameterName=BacklightAperture, sortPriority=17}
  n21 = DotProduct(A=@n15, B=@n16)
  n22 = Multiply(A=Multiply(A=Power(Base=@n19, Exp=@n20), B=Saturate(None=@n21)), B=NamedRerouteUsage())
  n23 = VectorParameter() {group=Crystal, parameterName=BacklightColorMultiply, sortPriority=13, defaultValue=<rgba 0.4136 2 1.1092 1>}
  n24 = LinearInterpolate(A=@n23.RGB, Alpha=NamedRerouteUsage()) {constB=2}
  NamedRerouteDeclaration(None=Multiply(A=Multiply(A=@n22, B=@n24), B=ScalarParameter() {defaultValue=10, group=Crystal, parameterName=BackLightIntensity, sortPriority=14})) {name=Backlight, nodeColor=<rgba 1 0 0.2578 1>}
  NamedRerouteDeclaration(None=@n13.BaseColor) {name=BaseColor, nodeColor=<rgba 0.1661 0.3253 1 1>}
  n25 = ReflectionVectorWS()
  n26 = ScalarParameter() {defaultValue=0.1, group=Crystal, parameterName=ReflectionRoughness, sortPriority=22}
  n27 = VectorParameter() {group=Crystal, parameterName=ReflectionColor, sortPriority=21, defaultValue=<rgba 0.0212 1 0.7231 1>}
  n28 = NamedRerouteUsage()
  NamedRerouteDeclaration(None=Multiply(A=SkyLightEnvMapSample(Direction=@n25, Roughness=@n26), B=LinearInterpolate(A=@n27.RGB, Alpha=@n28) {constB=0.98})) {name=Reflection, nodeColor=<rgba 1 0.1406 0 1>}
  n29 = TextureCoordinate()
  n30 = NamedRerouteDeclaration(None=@n29) {name=UVCoord, nodeColor=<rgba 0.5937 1 0 1>}
  n31 = ScalarParameter() {defaultValue=0.4, group=Crystal, parameterName=InteriorUVTiling, sortPriority=4}
  n32 = Constant() {r=-10}
  n33 = TextureSampleParameter2D(UVs=Multiply(A=@n29, B=ScalarParameter() {defaultValue=1.625, group=Crystal, parameterName=DepthUVTiling, sortPriority=8})) {
   group=Crystal
   parameterName=DepthTexture
   samplerType=SAMPLERTYPE_LinearGrayscale
   sortPriority=9
   texture=/Game/Global/Textures/Default/Masks_and_Gradients/T_MF_Crystal2
  }
  n34 = LinearInterpolate(
   A=ScalarParameter() {group=Crystal, parameterName=DepthMinimum, sortPriority=6}
   B=ScalarParameter() {defaultValue=-21.344, group=Crystal, parameterName=DepthMaximum, sortPriority=7}
   Alpha=NamedRerouteDeclaration(None=@n33.R) {name=SparksDepth, nodeColor=<rgba 0.1156 1 0.3086 1>}
  ) {constA=-4, constB=-20}
  NamedRerouteDeclaration(None=BumpOffset(
    Coordinate=Multiply(A=@n30, B=@n31)
    Height=MaterialFunctionCall!(
     Low _ PC Low and Med=@n32
     "High _ XB1, PS4"=@n34
     "Epic _ PC High, XSX, PS5"=@n34
     --------
     "Mobile Low _ PC Perf, Android"=@n32
     Mobile Med _ Switch=@n32
    ) {materialFunction=/Game/Global/MaterialFunctions/TechArt/MF_Fortnite_QualitySwitch_Vector3}.Result
   )) {name=UVOffsetCoord, nodeColor=<rgba 0.9453 0 1 1>}
  n35 = PixelNormalWS()
  n36 = CameraVectorWS()
  n37 = OneMinus(None=Saturate(None=DotProduct(A=@n35, B=@n36)))
  NamedRerouteDeclaration(None=Multiply(A=Multiply(A=@n37, B=@n37), B=ScalarParameter() {defaultValue=0.2, group=Crystal, parameterName=FresnelIntensity, sortPriority=20})) {name=Fresnel, nodeColor=<rgba 0 0.8828 1 1>}
  n38 = TextureCoordinate()
  n39 = Constant2Vector() {g=0.2, r=0.75}
  n40 = Constant() {r=0.06}
  n41 = SphereMask(A=TextureCoordinate(), B=Constant2Vector() {g=0.65, r=0.8}, Radius=Constant() {r=0.06})
  n42 = SphereMask(A=TextureCoordinate(), B=Constant2Vector() {g=0.5}, Radius=Constant() {r=0.06})
  NamedRerouteDeclaration(None=Max(A=SphereMask(A=@n38, B=@n39, Radius=@n40), B=Max(A=@n41, B=@n42))) {name=SmallDiamondMask, nodeColor=<rgba 1 0 0.7266 1>}
 ]}
