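# Material function for the SilverCanine backpack that blends animated sequins onto incoming material attributes using the TechArtMask input.
# Rotatable, tiled UVs drive sequin offset and normal textures, shaping brightness, two-color falloff and a normal detail added to the base.
# Sparkle emissive multiplies time-scrolled noise by a panning caustics gradient, tinted and brightened.
# Base color, metallic, roughness, emissive and normal are each lerped by the mask.
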MaterialFunction /BRCosmetics/Accessories/FORT_Backpacks/Backpack_F_MED_SilverCanine/Materials/MF_SilverCanine_Sequins
graph {nodes [
  n1 = FunctionInput() {bUsePreviewValueAsDefault=true, inputName=inMat, inputType=FunctionInput_MaterialAttributes, sortPriority=-1}
  n2 = GetMaterialAttributes(None=@n1) {attributeGetTypes [
    #1
    #2
    #3
    #4
   ]}
  n3 = GetMaterialAttributes(None=NamedRerouteDeclaration(None=@n1) {name=Mat_IN, nodeColor=<rgba 1 0.0469 0 1>}) {attributeGetTypes [
    #5
    #4
    #1
    #6
    #2
    #3
   ]}
  n4 = FunctionInput(Preview=Constant() {r=1}) {inputName=TechArtMask, inputType=FunctionInput_Scalar, sortPriority=8}
  n5 = SetMaterialAttributes!(
   MaterialAttributes=@n2.MaterialAttributes
   BaseColor=LinearInterpolate(A=@n3.BaseColor, B=NamedRerouteUsage(), Alpha=@n4)
   Metallic=LinearInterpolate(A=@n3.Metallic, B=NamedRerouteUsage(), Alpha=@n4)
   Roughness=LinearInterpolate(A=@n3.Roughness, B=NamedRerouteUsage(), Alpha=@n4)
   EmissiveColor=LinearInterpolate(A=@n3.EmissiveColor, B=NamedRerouteUsage(), Alpha=@n4)
   Normal=LinearInterpolate(A=@n3.Normal, B=NamedRerouteUsage(), Alpha=@n4)
   Specular
  ) {attributeSetTypes [
    #5
    #4
    #1
    #6
    #2
    #3
   ]}
  FunctionOutput(None=@n5)
  n6 = TextureCoordinate()
  n7 = Constant()
  n8 = FunctionInput(Preview=Constant()) {bUsePreviewValueAsDefault=true, inputName=SequinRotationAngle, inputType=FunctionInput_Scalar}
  NamedRerouteDeclaration(None=Multiply(
    A=MaterialFunctionCall(UVs=@n6, Rotation Center=@n7, "Rotation Angle (0-1)"=@n8) {materialFunction=&CustomRotator}.Rotated Values
    B=ScalarParameter() {defaultValue=32, group=Sequin, parameterName=SequinTile, sortPriority=5}
   )) {name=SequinUVs, nodeColor=<rgba 0 1 0.0859 1>}
  n9 = TextureSample!(
   UVs=NamedRerouteUsage()
   Tex=FunctionInput(Preview=TextureObject() {samplerType=SAMPLERTYPE_LinearGrayscale, texture=/Game/Global/Textures/Default/Tileable_Patterns/T_SequinTile}) {bUsePreviewValueAsDefault=true, inputName=SequinOffset, inputType=FunctionInput_Texture2D}
   Bias=ScalarParameter() {group=Sequin, parameterName=SequinTextureMipBias}
   Apply View MipBias
  ) {
   mipValueMode=TMVM_MipBias
   samplerSource=SSM_Wrap_WorldGroupSettings
   samplerType=SAMPLERTYPE_LinearGrayscale
   texture=&T_SequinTile
  }
  NamedRerouteDeclaration(None=@n9.RGB) {name=SequinOffset, nodeColor=<rgba 1 0 0.6562 1>}
  n10 = NamedRerouteUsage()
  n11 = ScalarParameter() {group=Sequin, parameterName=SequinFresnel, sortPriority=7}
  n12 = OneMinus(None=@n11)
  n13 = LinearInterpolate(Alpha=Multiply(A=@n12, B=@n11)) {constA=0.7}
  NamedRerouteDeclaration(None=Multiply(A=ScalarParameter() {defaultValue=1.5, group=Sequin, parameterName=SequinBrightness}, B=Multiply(A=@n10, B=@n13))) {name=BaseColor, nodeColor=<rgba 0 0.6719 1 1>}
  NamedRerouteDeclaration(None=@n2.Metallic) {name=BaseMetallic, nodeColor=<rgba 1 0.3984 0 1>}
  NamedRerouteDeclaration(None=LinearInterpolate(
    A=ScalarParameter() {defaultValue=0.9, parameterName=SequinMetalness}
    B=NamedRerouteUsage()
    Alpha=ScalarParameter() {defaultValue=1, group=Sequin, parameterName=SequinBaseRoughnessBlendAmount}
   )) {name=Metalness_Out, nodeColor=<rgba 0.0781 0 1 1>}
  n14 = VectorParameter() {group=Sequin, parameterName=SequinFallofColor01, defaultValue=<rgba 0.8842 0.6282 1 1>}
  n15 = VectorParameter() {group=Sequin, parameterName=SequinFallofColor02, defaultValue=<rgba 0.6663 0.575 1 1>}
  n16 = LinearInterpolate(Alpha=NamedRerouteUsage()) {constA=0.5, constB=1.5}
  NamedRerouteDeclaration(None=LinearInterpolate(
    A=FunctionInput(Preview=@n14.RGB) {bUsePreviewValueAsDefault=true, inputName=SequinFalloffColor01}
    B=FunctionInput(Preview=@n15.RGB) {bUsePreviewValueAsDefault=true, inputName=SequinFalloffColor02}
    Alpha=Multiply(A=@n16, B=@n12)
   )) {name=SequinColor, nodeColor=<rgba 0.8984 0 1 1>}
  NamedRerouteDeclaration(None=@n2.Normal) {name=BaseNormals, nodeColor=<rgba 0 1 0.1562 1>}
  n17 = TextureSample(
   UVs=NamedRerouteUsage()
   Tex=FunctionInput(Preview=TextureObject() {samplerType=SAMPLERTYPE_Normal, texture=/Game/Global/Textures/Default/Tileable_Patterns/T_SequinTile_N}) {bUsePreviewValueAsDefault=true, inputName=SequinNormal, inputType=FunctionInput_Texture2D}
   Bias=ScalarParameter() {group=Sequin, parameterName=SequinTextureMipBias}
  ) {
   mipValueMode=TMVM_MipBias
   samplerSource=SSM_Wrap_WorldGroupSettings
   samplerType=SAMPLERTYPE_Normal
   texture=&T_SequinTile_N
  }
  n18 = Multiply(
   A=ScalarParameter() {defaultValue=0.8, group=Sequin, parameterName=SequinNormalIntensity, sortPriority=15}
   B=Constant3Vector() {constant=<rgba 1 1 0 0>}
  )
  NamedRerouteDeclaration(None=Add(A=NamedRerouteUsage(), B=Multiply(A=@n17.RGB, B=@n18)) {constB=0}) {name=Normal_Out, nodeColor=<rgba 0.2891 0 1 1>}
  n19 = NamedRerouteUsage()
  n20 = Time()
  n21 = ScalarParameter() {defaultValue=0.01, group=Sequin, parameterName=SequinSparkleSpeed, sortPriority=12}
  n22 = TextureSample!(UVs=Add(A=ComponentMask(None=@n19) {g=true, r=true}, B=Multiply(A=@n20, B=@n21) {constB=0.2}), Tex, Apply View MipBias) {samplerSource=SSM_Wrap_WorldGroupSettings, samplerType=SAMPLERTYPE_Normal, texture=/Game/Global/Textures/Default/Noise/64x64Noise}
  n23 = TextureCoordinate()
  n24 = ScalarParameter() {defaultValue=2, group=Sequin, parameterName=SequinEmissiveTiling, sortPriority=14}
  n25 = VectorParameter() {group=Sequin, parameterName=SequinEmissivePanSpeed, sortPriority=14, defaultValue=<rgba 0.05 0.05 0 0>}
  n26 = TextureSampleParameter2D(UVs=Panner(Coordinate=Multiply(A=@n23, B=@n24), Speed=ComponentMask(None=@n25.RGB) {g=true, r=true}) {speedX=0.05, speedY=0.05}) {
   group=Sequin
   parameterName=SequinEmissiveGradient
   samplerSource=SSM_Wrap_WorldGroupSettings
   sortPriority=14
   texture=/Game/Effects/Fort_Effects/Textures/Tile/T_ShieldCaustics_Tile_Chromatic
  }
  n27 = Multiply(
   A=Power(Base=@n22.R) {constExponent=4}
   B=Power(Base=ChannelMaskParameter(None=@n26.RGB) {group=Sequin, parameterName=SequinEmissiveGradientChannel, sortPriority=14}.R)
  )
  n28 = VectorParameter() {group=Sequin, parameterName=SparkleColor, defaultValue=<rgba 0.5742 0.228 1 1>}
  NamedRerouteDeclaration(None=Multiply(A=Multiply(A=@n27, B=@n28.RGB), B=ScalarParameter() {defaultValue=10, group=Sequin, parameterName=SparkleBRightness, sortPriority=13})) {name=Emissive_Out, nodeColor=<rgba 0.0312 1 0 1>}
  NamedRerouteDeclaration(None=Constant() {r=0.3}) {name=Roughness, nodeColor=<rgba 1 0 0.3516 1>}
 ]}
