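# Parent skeletal-mesh material for the Silver Canine backpack cosmetic, adding sequin and gemstone looks between the 2023 character PreFX and PostFX functions.
# A sequin function uses a selectable tech-art mask channel, with tile textures, falloff colors and rotation.
# The gemstone branch runs thin-film and crystal functions, with the red mask channel declared as its mask.
# A quality switch limits its bump-offset triangle normal detail to higher tiers.
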Material /BRCosmetics/Accessories/FORT_Backpacks/Backpack_F_MED_SilverCanine/Materials/M_SilverCanine_Backpack_Parent
graph {
 properties {
  bUseMaterialAttributes=true
  bUsedWithMeshDeformer=true
  bUsedWithMorphTargets=true
  bUsedWithSkeletalMesh=true
  pixelDepthOffsetMode=PDOM_Legacy
  usedShadingModels=Default Lit
 }
 nodes [
  n1 = NamedRerouteUsage()
  n2 = TextureObjectParameter() {
   group=Sequin
   parameterName=SequinOffest
   samplerType=SAMPLERTYPE_LinearGrayscale
   sortPriority=1
   texture=/Game/Global/Textures/Default/Tileable_Patterns/T_SequinTile
  }
  n3 = TextureObjectParameter() {
   group=Sequin
   parameterName=SequinNormal
   samplerType=SAMPLERTYPE_Normal
   sortPriority=2
   texture=/Game/Global/Textures/Default/Tileable_Patterns/T_SequinTile_N
  }
  n4 = VectorParameter() {group=Sequin, parameterName=SequinFalloffColor01, sortPriority=3, defaultValue=<rgba 0.8542 0.387 0.5588 1>}
  n5 = VectorParameter() {group=Sequin, parameterName=SequinFalloffColor02, sortPriority=4, defaultValue=<rgba 0.3993 0.2651 0.8438 1>}
  n6 = ScalarParameter() {group=Sequin, parameterName=SequinRotationAngle}
  n7 = MaterialFunctionCall() {materialFunction=&MF_CharacterParent_2023_PreFX}
  n8 = ChannelMaskParameter(None=@n7.Tech Art Mask) {group=Sequin, parameterName=SequinChannel, sortPriority=0}
  MaterialFunctionCall(
   Material=BlendMaterialAttributes(
    A=MaterialFunctionCall!(
     inMat=@n1
     SequinOffset=@n2
     SequinNormal=@n3
     SequinFalloffColor01=@n4.RGB
     SequinFalloffColor02=@n5.RGB
     SequinRotationAngle=@n6
     TechArtMask=@n8.R
    ) {materialFunction=&/MF_SilverCanine_Sequins}.Result
    B=NamedRerouteUsage()
    Alpha=NamedRerouteUsage()
   )
   M=@n7.M
   Pre Skinned Local Position=@n7.M
   Distance Blend=@n7.M
   Fresnel=@n7.M
  ) {materialFunction=&MF_CharacterParent_2023_PostFX}
  NamedRerouteDeclaration(None=ComponentMask(None=@n7.Tech Art Mask) {r=true}) {name=GemstoneMask, nodeColor=<rgba 1 0.7266 0 1>}
  n9 = GetMaterialAttributes(None=NamedRerouteUsage()) {attributeGetTypes [
    #1
    #2
    #3
    #4
   ]}
  n10 = BumpOffset(Coordinate=Multiply(A=TextureCoordinate(), B=Constant() {r=-2.97}), Height=Constant() {r=-75.2159}) {heightRatio=0.01}
  n11 = MaterialFunctionCall(
   BaseNormal=@n9.Normal
   AdditionalNormal=MaterialFunctionCall(
    Normal=TextureSample(UVs=@n10) {samplerType=SAMPLERTYPE_Normal, texture=/Game/Global/Materials/MaterialGym/Materials/TechArtPatterns/Textures/T_Triangles_N}.RGB
    Flatness=ScalarParameter() {defaultValue=0.85, group=SliverGanineGemstone, parameterName=GemstoneNormalFlatness}
   ) {materialFunction=&FlattenNormal}.Result
  ) {materialFunction=&BlendAngleCorrectedNormals}
  n12 = MakeMaterialAttributes!(
   BaseColor=@n9.Base Color
   Metallic=Constant() {r=0.3}
   Specular=@n9.Base Color
   Roughness=@n9.Base Color
   Anisotropy
   EmissiveColor
   Opacity
   OpacityMask
   Normal=MaterialFunctionCall!(
    Low _ PC Low and Med=@n9.Normal
    "High _ XB1, PS4"=@n11.Result
    "Epic _ PC High, XSX, PS5"=@n11.Result
    --------
    "Mobile Low _ PC Perf, Android"=@n9.Normal
    Mobile Med _ Switch=@n9.Normal
   ) {materialFunction=/Game/Global/MaterialFunctions/TechArt/MF_Fortnite_QualitySwitch_Vector3}.Result
   Tangent
   WorldPositionOffset
   SubsurfaceColor
   ClearCoat
   ClearCoatRoughness
   AmbientOcclusion
   Refraction
   CustomizedUVs_0
   CustomizedUVs_1
   CustomizedUVs_2
   CustomizedUVs_3
   CustomizedUVs_4
   CustomizedUVs_5
   CustomizedUVs_6
   CustomizedUVs_7
   PixelDepthOffset
   ShadingModel
   Displacement
  )
  n13 = Multiply(A=Fresnel(ExponentIn=Constant() {r=1}) {baseReflectFraction=0}, B=Constant() {r=0.74})
  n14 = ScalarParameter() {defaultValue=0.3, group=SliverGanineGemstone, parameterName=ThinFilmAlpha}
  NamedRerouteDeclaration(None=MaterialFunctionCall!(Material=MaterialFunctionCall!(Material Attributes=@n12, Fresnel=@n13, Alpha=@n14) {materialFunction=&/MF_ThinFilm_SilverCanine}.Result) {materialFunction=&/MF_Crystal_SilverCanine}.Result) {name=Gemstone, nodeColor=<rgba 1 0 0.2344 1>}
  NamedRerouteDeclaration(None=@n7.Result) {name=InputMA, nodeColor=<rgba 0 0.5547 1 1>}
 ]
}
